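; Gameplay ability for the Battle Pass S20 "Purple Overheat" quest that reports overheat progress to the owning player's quest manager.
; On activation from an event, it runs the parent logic, then checks whether the player's current weapon data tags match any of the configured weapon tags.
; On a match, it sends a complex custom stat event tagged Athena.Quests.Purple.Overheat with a count of one, then ends the ability.
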(event Ability|EventActivateAbilityFromEvent
  (params (EventData :type Gameplay%20Event%20Data%20Structure%20%28by%20ref%29))
  ;[1 1 1] Activate ability
  (|CallParentFunction EventData)
  (if (GameplayTags|HasAnyTags (Variables|Default|GetWeaponTags) (Tags|GetWeaponDataGameplayTags (Weapon|GetCurrentWeapon (Variables|GAVars|Data|GetOwningPlayerPawn))) true)
    (Quest|SendComplexCustomStatEvent (Variables|GAVars|Data|GetOwningPlayerQuestManager) 0 "(GameplayTags=)" "(GameplayTags=((TagName=\"Athena.Quests.Purple.Overheat\")))" 1)
    (Ability|EndAbility)))
